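# Sound cue for the ready sound of the Taxi pickaxe cosmetic in Battle Royale, reusing the Taxi swing recordings at low volume.
# An equally weighted Random node picks one of six swing waves and feeds a Modulator that raises pitch and slightly varies volume.
# It plays under the melee weapon sound class with close-range attenuation, a pickaxe-ready property sheet, load-time priming, and a concurrency override that blocks new instances at the limit.
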SoundCue /BRCosmetics/Sounds/PickAxes/Taxi/PA_Taxi_Ready_Athena_Cue
root=Modulator_0
nodes [
 {id=Modulator_0, children [Random_1], properties {pitchMax=1.2, pitchMin=1.1, volumeMax=1.1, volumeMin=1}}
 {id=Random_1, children [WavePlayer_1,WavePlayer_3,WavePlayer_6,WavePlayer_7,WavePlayer_2,WavePlayer_4], properties {weights [1,1,1,1,1,1]}}
 {id=WavePlayer_1, properties {soundWaveAssetPtr=&/PA_Taxi_Swing_01}}
 {id=WavePlayer_2, properties {soundWaveAssetPtr=&/PA_Taxi_Swing_05}}
 {id=WavePlayer_3, properties {soundWaveAssetPtr=&/PA_Taxi_Swing_02}}
 {id=WavePlayer_4, properties {soundWaveAssetPtr=&/PA_Taxi_Swing_06}}
 {id=WavePlayer_6, properties {soundWaveAssetPtr=&/PA_Taxi_Swing_03}}
 {id=WavePlayer_7, properties {soundWaveAssetPtr=&/PA_Taxi_Swing_04}}
]
properties {
 attenuationSettings=&Fort_Weapon_Projectile_Close_Athena_Att
 audioPropertiesSheet=/BRCosmetics/Sounds/Settings/PropertySheets/Pickaxe/APS_Cosmetic_Pickaxe_Ready
 bOverrideConcurrency=true
 bPrimeOnLoad=true
 maxDistance=4950
 soundClassObject=&Fort_WeaponMelee
 volumeMultiplier=0.2
 concurrencyOverrides {resolutionRule=PreventNew}
}
